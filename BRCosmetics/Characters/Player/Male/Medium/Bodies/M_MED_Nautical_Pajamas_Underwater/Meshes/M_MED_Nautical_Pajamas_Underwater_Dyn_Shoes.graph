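# Physics asset for the dynamic shoe parts of the male medium Nautical Pajamas Underwater cosmetic: fish eyes, fins and a dangling light on each foot.
# Each part has a small MSTR capsule body constrained to its foot's ball bone with free swing and twist.
# Eye capsules and fin boxes hang off their MSTR with tight angular limits; each light is a two-link capsule chain, with slightly wider swing on the right.
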PhysicsAsset /BRCosmetics/Characters/Player/Male/Medium/Bodies/M_MED_Nautical_Pajamas_Underwater/Meshes/M_MED_Nautical_Pajamas_Underwater_Dyn_Shoes
bodies {
 ball_l [{shapeType=Capsule, length=1, radius=1}]
 ball_r [{shapeType=Capsule, length=1, radius=1}]
 dyn_fins_01_l [{shapeType=Box, extentX=6, extentY=5, extentZ=1, center=<3 0 0>}]
 dyn_fins_01_l_MSTR [{shapeType=Capsule, length=0.5, radius=0.5}]
 dyn_fins_01_r [{shapeType=Box, extentX=6, extentY=5, extentZ=1, center=<3 0 0>}]
 dyn_fins_01_r_MSTR [{shapeType=Capsule, length=0.5, radius=0.5}]
 dyn_fins_02_l [{shapeType=Box, extentX=6, extentY=5, extentZ=1, center=<3 0 0>}]
 dyn_fins_02_l_MSTR [{shapeType=Capsule, length=0.5, radius=0.5}]
 dyn_fins_02_r [{shapeType=Box, extentX=6, extentY=5, extentZ=1, center=<3 0 0>}]
 dyn_fins_02_r_MSTR [{shapeType=Capsule, length=0.5, radius=0.5}]
 dyn_fish_eye_01_l [{shapeType=Capsule, length=2, radius=2.5, center=<2 0 0>, rotation=<rot 90 0 0>}]
 dyn_fish_eye_01_l_MSTR [{shapeType=Capsule, length=0.5, radius=0.5}]
 dyn_fish_eye_01_r [{shapeType=Capsule, length=2, radius=2.5, center=<2 0 0>, rotation=<rot 90 0 0>}]
 dyn_fish_eye_01_r_MSTR [{shapeType=Capsule, length=0.5, radius=0.5}]
 dyn_fish_eye_02_l [{shapeType=Capsule, length=2, radius=2.5, center=<2 0 0>, rotation=<rot 90 0 0>}]
 dyn_fish_eye_02_l_MSTR [{shapeType=Capsule, length=0.5, radius=0.5}]
 dyn_fish_eye_02_r [{shapeType=Capsule, length=2, radius=2.5, center=<2 0 0>, rotation=<rot 90 0 0>}]
 dyn_fish_eye_02_r_MSTR [{shapeType=Capsule, length=0.5, radius=0.5}]
 dyn_light_l_01 [{shapeType=Capsule, length=1, radius=1, center=<1 0 0>, rotation=<rot 90 0 0>}]
 dyn_light_l_02 [{shapeType=Capsule, length=1, radius=1.5, center=<3.5 0 0>, rotation=<rot 79.9997 0 0>}]
 dyn_light_l_MSTR [{shapeType=Capsule, length=0.5, radius=0.5}]
 dyn_light_r_01 [{shapeType=Capsule, length=1, radius=1, center=<1 0 0>, rotation=<rot 90 0 0>}]
 dyn_light_r_02 [{shapeType=Capsule, length=1, radius=1.5, center=<3.5 0 0>, rotation=<rot 79.9997 0 0>}]
 dyn_light_r_MSTR [{shapeType=Capsule, length=0.5, radius=0.5}]
}
constraints [
 {
  bone1Name=dyn_light_l_MSTR
  bone2Name=ball_l
  swing1LimitDegrees=45
  swing1Motion=Free
  swing2LimitDegrees=45
  swing2Motion=Free
  twistLimitDegrees=45
  twistMotion=Free
 }
 {bone1Name=dyn_light_l_01, bone2Name=dyn_light_l_MSTR, swing1LimitDegrees=15, swing2LimitDegrees=15, twistLimitDegrees=5}
 {bone1Name=dyn_light_l_02, bone2Name=dyn_light_l_01, swing1LimitDegrees=15, swing2LimitDegrees=15, twistLimitDegrees=5}
 {
  bone1Name=dyn_fish_eye_01_l_MSTR
  bone2Name=ball_l
  swing1LimitDegrees=45
  swing1Motion=Free
  swing2LimitDegrees=45
  swing2Motion=Free
  twistLimitDegrees=45
  twistMotion=Free
 }
 {bone1Name=dyn_fish_eye_01_l, bone2Name=dyn_fish_eye_01_l_MSTR, swing1LimitDegrees=20, swing2LimitDegrees=20, twistLimitDegrees=5}
 {
  bone1Name=dyn_fish_eye_02_l_MSTR
  bone2Name=ball_l
  swing1LimitDegrees=45
  swing1Motion=Free
  swing2LimitDegrees=45
  swing2Motion=Free
  twistLimitDegrees=45
  twistMotion=Free
 }
 {bone1Name=dyn_fish_eye_02_l, bone2Name=dyn_fish_eye_02_l_MSTR, swing1LimitDegrees=20, swing2LimitDegrees=20, twistLimitDegrees=5}
 {
  bone1Name=dyn_fins_01_l_MSTR
  bone2Name=ball_l
  swing1LimitDegrees=45
  swing1Motion=Free
  swing2LimitDegrees=45
  swing2Motion=Free
  twistLimitDegrees=45
  twistMotion=Free
 }
 {bone1Name=dyn_fins_01_l, bone2Name=dyn_fins_01_l_MSTR, swing1LimitDegrees=20, swing2LimitDegrees=20, twistLimitDegrees=2}
 {
  bone1Name=dyn_fins_02_l_MSTR
  bone2Name=ball_l
  swing1LimitDegrees=45
  swing1Motion=Free
  swing2LimitDegrees=45
  swing2Motion=Free
  twistLimitDegrees=45
  twistMotion=Free
 }
 {bone1Name=dyn_fins_02_l, bone2Name=dyn_fins_02_l_MSTR, swing1LimitDegrees=20, swing2LimitDegrees=20, twistLimitDegrees=2}
 {
  bone1Name=dyn_light_r_MSTR
  bone2Name=ball_r
  swing1LimitDegrees=45
  swing1Motion=Free
  swing2LimitDegrees=45
  swing2Motion=Free
  twistLimitDegrees=45
  twistMotion=Free
 }
 {bone1Name=dyn_light_r_01, bone2Name=dyn_light_r_MSTR, swing1LimitDegrees=20, swing2LimitDegrees=15, twistLimitDegrees=5}
 {bone1Name=dyn_light_r_02, bone2Name=dyn_light_r_01, swing1LimitDegrees=20, swing2LimitDegrees=15, twistLimitDegrees=5}
 {
  bone1Name=dyn_fish_eye_01_r_MSTR
  bone2Name=ball_r
  swing1LimitDegrees=45
  swing1Motion=Free
  swing2LimitDegrees=45
  swing2Motion=Free
  twistLimitDegrees=45
  twistMotion=Free
 }
 {bone1Name=dyn_fish_eye_01_r, bone2Name=dyn_fish_eye_01_r_MSTR, swing1LimitDegrees=20, swing2LimitDegrees=20, twistLimitDegrees=5}
 {
  bone1Name=dyn_fish_eye_02_r_MSTR
  bone2Name=ball_r
  swing1LimitDegrees=45
  swing1Motion=Free
  swing2LimitDegrees=45
  swing2Motion=Free
  twistLimitDegrees=45
  twistMotion=Free
 }
 {bone1Name=dyn_fish_eye_02_r, bone2Name=dyn_fish_eye_02_r_MSTR, swing1LimitDegrees=20, swing2LimitDegrees=20, twistLimitDegrees=5}
 {
  bone1Name=dyn_fins_01_r_MSTR
  bone2Name=ball_r
  swing1LimitDegrees=45
  swing1Motion=Free
  swing2LimitDegrees=45
  swing2Motion=Free
  twistLimitDegrees=45
  twistMotion=Free
 }
 {bone1Name=dyn_fins_01_r, bone2Name=dyn_fins_01_r_MSTR, swing1LimitDegrees=20, swing2LimitDegrees=20, twistLimitDegrees=2}
 {
  bone1Name=dyn_fins_02_r_MSTR
  bone2Name=ball_r
  swing1LimitDegrees=45
  swing1Motion=Free
  swing2LimitDegrees=45
  swing2Motion=Free
  twistLimitDegrees=45
  twistMotion=Free
 }
 {bone1Name=dyn_fins_02_r, bone2Name=dyn_fins_02_r_MSTR, swing1LimitDegrees=20, swing2LimitDegrees=20, twistLimitDegrees=2}
]
