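# Anim Blueprint for the Male Medium underwater Nautical Pajamas body cosmetic, adding secondary motion to its zipper and shoe fins and fish eyes.
# It copies the attached parent mesh's pose, then weighted constraints position the dynamic master bones from chest, neck, clavicle and foot bones.
# Three chained RigidBody nodes simulate each shoe and the zipper in base-bone space with custom gravity, reusing the base Nautical Pajamas zipper physics asset.
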AnimBlueprint /BRCosmetics/Characters/Player/Male/Medium/Bodies/M_MED_Nautical_Pajamas_Underwater/Meshes/M_MED_Nautical_Pajamas_Underwater_AnimBP
target_skeleton=&/M_MED_Nautical_Pajamas_Underwater_Skeleton
graphs [
 AnimGraph {nodes [
   n1 = "Constraint - dyn_zipper_MSTR (PPPPP)"!(
    ComponentPose=Local To Component(LocalPose=Copy Pose From Mesh() {bUseAttachedParent=true})
    Alpha=1
    ConstraintWeights_0=0.0242
    ConstraintWeights_1:float=0
    ConstraintWeights_2:float=0.001
    ConstraintWeights_3:float=0.9465
    ConstraintWeights_4:float=0.0283
   ) -> (Pose) {
    alphaBoolBlend {blendOption=Linear}
    boneToModify {boneName=dyn_zipper_MSTR}
    constraintSetup [
     {targetBone {boneName=clavicle_r}}
     {targetBone {boneName=neck_01}}
     {targetBone {boneName=spine_04}}
     {targetBone {boneName=spine_05}}
     {targetBone {boneName=clavicle_l}}
    ]
    constraintWeights [1,1,1,1,1]
   }
   n2 = "Constraint - dyn_fins_01_r_MSTR (PP)"!(ComponentPose=@n1, Alpha=1, ConstraintWeights_0=0.5645, ConstraintWeights_1:float=0.4355) -> (Pose) {
    alphaBoolBlend {blendOption=Linear}
    boneToModify {boneName=dyn_fins_01_r_MSTR}
    constraintSetup [{targetBone {boneName=ball_r}}, {targetBone {boneName=foot_r}}]
    constraintWeights [1,1]
   }
   n3 = "Constraint - dyn_fins_02_r_MSTR (PP)"!(ComponentPose=@n2, Alpha=1, ConstraintWeights_0=0.6076, ConstraintWeights_1:float=0.3924) -> (Pose) {
    alphaBoolBlend {blendOption=Linear}
    boneToModify {boneName=dyn_fins_02_r_MSTR}
    constraintSetup [{targetBone {boneName=ball_r}}, {targetBone {boneName=foot_r}}]
    constraintWeights [1,1]
   }
   n4 = "Constraint - dyn_fins_01_l_MSTR (PP)"!(ComponentPose=@n3, Alpha=1, ConstraintWeights_0=0.5645, ConstraintWeights_1:float=0.4355) -> (Pose) {
    alphaBoolBlend {blendOption=Linear}
    boneToModify {boneName=dyn_fins_01_l_MSTR}
    constraintSetup [{targetBone {boneName=ball_l}}, {targetBone {boneName=foot_l}}]
    constraintWeights [1,1]
   }
   n5 = "Constraint - dyn_fins_02_l_MSTR (PP)"!(ComponentPose=@n4, Alpha=1, ConstraintWeights_0=0.6076, ConstraintWeights_1:float=0.3924) -> (Pose) {
    alphaBoolBlend {blendOption=Linear}
    boneToModify {boneName=dyn_fins_02_l_MSTR}
    constraintSetup [{targetBone {boneName=ball_l}}, {targetBone {boneName=foot_l}}]
    constraintWeights [1,1]
   }
   n6 = "Constraint - dyn_fish_eye_01_r_MSTR (PP)"!(ComponentPose=@n5, Alpha=1, ConstraintWeights_0=0.8215, ConstraintWeights_1:float=0.1785) -> (Pose) {
    alphaBoolBlend {blendOption=Linear}
    boneToModify {boneName=dyn_fish_eye_01_r_MSTR}
    constraintSetup [{targetBone {boneName=ball_r}}, {targetBone {boneName=foot_r}}]
    constraintWeights [1,1]
   }
   n7 = "Constraint - dyn_fish_eye_02_r_MSTR (PP)"!(ComponentPose=@n6, Alpha=1, ConstraintWeights_0=0.8215, ConstraintWeights_1:float=0.1785) -> (Pose) {
    alphaBoolBlend {blendOption=Linear}
    boneToModify {boneName=dyn_fish_eye_02_r_MSTR}
    constraintSetup [{targetBone {boneName=ball_r}}, {targetBone {boneName=foot_r}}]
    constraintWeights [1,1]
   }
   n8 = "Constraint - dyn_fish_eye_01_l_MSTR (PP)"!(ComponentPose=@n7, Alpha=1, ConstraintWeights_0=0.8215, ConstraintWeights_1:float=0.1785) -> (Pose) {
    alphaBoolBlend {blendOption=Linear}
    boneToModify {boneName=dyn_fish_eye_01_l_MSTR}
    constraintSetup [{targetBone {boneName=ball_l}}, {targetBone {boneName=foot_l}}]
    constraintWeights [1,1]
   }
   n9 = "Constraint - dyn_fish_eye_02_l_MSTR (PP)"!(ComponentPose=@n8, Alpha=1, ConstraintWeights_0=0.8215, ConstraintWeights_1:float=0.1785) -> (Pose) {
    alphaBoolBlend {blendOption=Linear}
    boneToModify {boneName=dyn_fish_eye_02_l_MSTR}
    constraintSetup [{targetBone {boneName=ball_l}}, {targetBone {boneName=foot_l}}]
    constraintWeights [1,1]
   }
   n10 = RigidBody(ComponentPose=@n9) {
    bClampLinearTranslationLimitToRefPose=true
    bOverrideWorldGravity=true
    overridePhysicsAsset=&/M_MED_Nautical_Pajamas_Underwater_Dyn_Shoe_L
    simulationSpace=BaseBoneSpace
    alphaBoolBlend {blendOption=Linear}
    baseBoneRef {boneName=foot_l}
    componentAppliedLinearAccClamp=<3000 3000 3000>
    componentLinearAccScale=<0.4 0.4 0.4>
    componentLinearVelScale=<0.3 0.3 0.3>
    overrideWorldGravity {z=-580}
   }
   n11 = RigidBody(ComponentPose=@n10) {
    bClampLinearTranslationLimitToRefPose=true
    bOverrideWorldGravity=true
    overridePhysicsAsset=&/M_MED_Nautical_Pajamas_Underwater_Dyn_Shoes
    simulationSpace=BaseBoneSpace
    alphaBoolBlend {blendOption=Linear}
    baseBoneRef {boneName=foot_r}
    componentAppliedLinearAccClamp=<3000 3000 3000>
    componentLinearAccScale=<0.4 0.4 0.4>
    componentLinearVelScale=<0.3 0.3 0.3>
    overrideWorldGravity {z=-580}
   }
   n12 = RigidBody(ComponentPose=@n11) {
    bClampLinearTranslationLimitToRefPose=true
    bOverrideWorldGravity=true
    overridePhysicsAsset=/BRCosmetics/Characters/Player/Male/Medium/Bodies/M_MED_Nautical_Pajamas/Meshes/M_MED_Nautical_Pajamas_Dyn_Zipper
    simulationSpace=BaseBoneSpace
    alphaBoolBlend {blendOption=Linear}
    baseBoneRef {boneName=spine_05}
    componentAppliedLinearAccClamp=<2000 2000 2000>
    componentLinearAccScale=<0.35 0.35 0.35>
    componentLinearVelScale=<0.3 0.3 0.3>
    overrideWorldGravity {z=-980}
   }
   Output Pose(Result=Component To Local(ComponentPose=@n12)) {layerGroup=None}
  ]}
 EventGraph {nodes [Event BlueprintUpdateAnimation()]}
]
variables []
