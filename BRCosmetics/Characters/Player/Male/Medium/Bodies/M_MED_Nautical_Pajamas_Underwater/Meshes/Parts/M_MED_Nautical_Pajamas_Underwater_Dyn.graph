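# Physics asset for secondary motion on the dynamic head parts of the male medium Nautical Pajamas Underwater cosmetic body: eyes, three fin pairs and a light.
# Small MSTR anchor capsules connect each part to a head capsule through joints with free swing and twist.
# Each eye and fin hangs from its anchor with tight angular limits.
# The light forms a three-link chain, stiffer on one swing axis.
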PhysicsAsset /BRCosmetics/Characters/Player/Male/Medium/Bodies/M_MED_Nautical_Pajamas_Underwater/Meshes/Parts/M_MED_Nautical_Pajamas_Underwater_Dyn
bodies {
 dyn_eyes_l [{shapeType=Capsule, length=1.25, radius=3.25, center=<3.5 0 0>, rotation=<rot 90 0 0>}]
 dyn_eyes_l_MSTR [{shapeType=Capsule, length=0.5, radius=0.5}]
 dyn_eyes_r [{shapeType=Capsule, length=1.25, radius=3.25, center=<3.5 0 0>, rotation=<rot 90 0 0>}]
 dyn_eyes_r_MSTR [{shapeType=Capsule, length=0.5, radius=0.5}]
 dyn_fins_01_l [{shapeType=Capsule, length=1, radius=1, center=<-0.5 2 0>, rotation=<rot 90 0 0>}]
 dyn_fins_01_l_MSTR [{shapeType=Capsule, length=0.5, radius=0.5}]
 dyn_fins_01_r [{shapeType=Capsule, length=1, radius=1, center=<-0.5 -2 0>, rotation=<rot 90 0 0>}]
 dyn_fins_01_r_MSTR [{shapeType=Capsule, length=0.5, radius=0.5}]
 dyn_fins_02_l [{shapeType=Capsule, length=1, radius=1, center=<0.5 2 0>, rotation=<rot 90 0 0>}]
 dyn_fins_02_l_MSTR [{shapeType=Capsule, length=0.5, radius=0.5}]
 dyn_fins_02_r [{shapeType=Capsule, length=1, radius=1, center=<0.2 -2.5 0>, rotation=<rot 90 0 0>}]
 dyn_fins_02_r_MSTR [{shapeType=Capsule, length=0.5, radius=0.5}]
 dyn_fins_03_l [{shapeType=Capsule, length=1, radius=1, center=<1 2 0>, rotation=<rot 90 0 0>}]
 dyn_fins_03_l_MSTR [{shapeType=Capsule, length=0.5, radius=0.5}]
 dyn_fins_03_r [{shapeType=Capsule, length=1, radius=1, center=<1 -2 0>, rotation=<rot 90 0 0>}]
 dyn_fins_03_r_MSTR [{shapeType=Capsule, length=0.5, radius=0.5}]
 dyn_light_01 [{shapeType=Capsule, length=2.75, radius=1, center=<1.7 0 0.5>, rotation=<rot 80 0 0>}]
 dyn_light_02 [{shapeType=Capsule, length=3, radius=1, center=<2.1782 0 0>, rotation=<rot 90 0 0>}]
 dyn_light_03 [{shapeType=Capsule, length=0.75, radius=2.5, center=<4.5001 -1.465 -0.1702>, rotation=<rot 90 0 0>}]
 dyn_light_MSTR [{shapeType=Capsule, length=0.5, radius=0.5}]
 head [{shapeType=Capsule, length=4.7645, radius=11.2817, center=<4.3195 0.4384 -0.0001>, rotation=<rot 90 -14.0363 -34.0392>}]
}
constraints [
 {
  bone1Name=dyn_fins_01_r_MSTR
  bone2Name=head
  swing1LimitDegrees=45
  swing1Motion=Free
  swing2LimitDegrees=45
  swing2Motion=Free
  twistLimitDegrees=45
  twistMotion=Free
 }
 {bone1Name=dyn_fins_01_r, bone2Name=dyn_fins_01_r_MSTR, swing1LimitDegrees=20, swing2LimitDegrees=20, twistLimitDegrees=5}
 {
  bone1Name=dyn_fins_02_r_MSTR
  bone2Name=head
  swing1LimitDegrees=45
  swing1Motion=Free
  swing2LimitDegrees=45
  swing2Motion=Free
  twistLimitDegrees=45
  twistMotion=Free
 }
 {bone1Name=dyn_fins_02_r, bone2Name=dyn_fins_02_r_MSTR, swing1LimitDegrees=20, swing2LimitDegrees=20, twistLimitDegrees=5}
 {
  bone1Name=dyn_fins_03_r_MSTR
  bone2Name=head
  swing1LimitDegrees=45
  swing1Motion=Free
  swing2LimitDegrees=45
  swing2Motion=Free
  twistLimitDegrees=45
  twistMotion=Free
 }
 {
  bone1Name=dyn_eyes_l_MSTR
  bone2Name=head
  swing1LimitDegrees=45
  swing1Motion=Free
  swing2LimitDegrees=45
  swing2Motion=Free
  twistLimitDegrees=45
  twistMotion=Free
 }
 {bone1Name=dyn_eyes_l, bone2Name=dyn_eyes_l_MSTR, swing1LimitDegrees=20, swing2LimitDegrees=20, twistLimitDegrees=2}
 {
  bone1Name=dyn_fins_01_l_MSTR
  bone2Name=head
  swing1LimitDegrees=45
  swing1Motion=Free
  swing2LimitDegrees=45
  swing2Motion=Free
  twistLimitDegrees=45
  twistMotion=Free
 }
 {bone1Name=dyn_fins_01_l, bone2Name=dyn_fins_01_l_MSTR, swing1LimitDegrees=20, swing2LimitDegrees=20, twistLimitDegrees=5}
 {
  bone1Name=dyn_fins_02_l_MSTR
  bone2Name=head
  swing1LimitDegrees=45
  swing1Motion=Free
  swing2LimitDegrees=45
  swing2Motion=Free
  twistLimitDegrees=45
  twistMotion=Free
 }
 {bone1Name=dyn_fins_02_l, bone2Name=dyn_fins_02_l_MSTR, swing1LimitDegrees=20, swing2LimitDegrees=20, twistLimitDegrees=5}
 {
  bone1Name=dyn_fins_03_l_MSTR
  bone2Name=head
  swing1LimitDegrees=45
  swing1Motion=Free
  swing2LimitDegrees=45
  swing2Motion=Free
  twistLimitDegrees=45
  twistMotion=Free
 }
 {bone1Name=dyn_fins_03_l, bone2Name=dyn_fins_03_l_MSTR, swing1LimitDegrees=20, swing2LimitDegrees=20, twistLimitDegrees=5}
 {bone1Name=dyn_fins_03_r, bone2Name=dyn_fins_03_r_MSTR, swing1LimitDegrees=20, swing2LimitDegrees=20, twistLimitDegrees=5}
 {
  bone1Name=dyn_light_MSTR
  bone2Name=head
  swing1LimitDegrees=45
  swing1Motion=Free
  swing2LimitDegrees=45
  swing2Motion=Free
  twistLimitDegrees=45
  twistMotion=Free
 }
 {bone1Name=dyn_light_01, bone2Name=dyn_light_MSTR, swing1LimitDegrees=5, swing2LimitDegrees=20, twistLimitDegrees=2}
 {bone1Name=dyn_light_02, bone2Name=dyn_light_01, swing1LimitDegrees=5, swing2LimitDegrees=20, twistLimitDegrees=2}
 {bone1Name=dyn_light_03, bone2Name=dyn_light_02, swing1LimitDegrees=5, swing2LimitDegrees=20, twistLimitDegrees=2}
 {
  bone1Name=dyn_eyes_r_MSTR
  bone2Name=head
  swing1LimitDegrees=45
  swing1Motion=Free
  swing2LimitDegrees=45
  swing2Motion=Free
  twistLimitDegrees=45
  twistMotion=Free
 }
 {bone1Name=dyn_eyes_r, bone2Name=dyn_eyes_r_MSTR, swing1LimitDegrees=20, swing2LimitDegrees=20, twistLimitDegrees=2}
]
